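# Material function that composites an animated flipbook mouth onto the BatterBoi character's material attributes.
# UV0 or UV2, chosen by static switch, is scaled and shifted by parameters, then samples a 4x4 SubUV atlas at the flipbook_face_mouth_index frame.
# Powered texture alpha times two per-axis SphereMasks on the SubUV coordinates blends the mouth into BaseColor and optionally Emissive.
# Mouth color and alpha are also separate outputs.
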MaterialFunction /BRCosmetics/Characters/Player/Male/Large/Bodies/M_LRG_BatterBoi/Materials/MF_FV2_Flipbook_BatterBoi
graph {nodes [
  Constant()
  Constant() {r=4}
  n1 = GetMaterialAttributes(None=FunctionInput() {bUsePreviewValueAsDefault=true, inputName=Material, inputType=FunctionInput_MaterialAttributes}) {attributeGetTypes [#1, #2]}
  n2 = TextureObjectParameter() {
   group=09_Flipbook Mouth
   parameterName=MouthFlipbook
   texture=/BRCosmetics/Characters/Player/Male/Large/Bodies/M_LRG_BatterBoi/Textures/T_Flipbook_BatterBoi
  }
  n3 = TextureCoordinate() {coordinateIndex=1}
  n4 = TextureCoordinate()
  n5 = MaterialFunctionCall(
   UVs=Multiply(
    A=StaticSwitchParameter(True=@n3, False=@n4) {group=09_Flipbook Mouth, parameterName=UseUV2forMouth}
    B=ScalarParameter() {defaultValue=2.5, group=09_Flipbook Mouth, parameterName=MouthScale}
   )
   Texture Scale=Constant() {r=0.8}
  ) {materialFunction=&ScaleUVsByCenter}
  n6 = AppendVector(
   A=ScalarParameter() {defaultValue=-2, group=09_Flipbook Mouth, parameterName=MouthUVShiftX}
   B=ScalarParameter() {defaultValue=0.1963, group=09_Flipbook Mouth, parameterName=MouthUVShiftY}
  )
  n7 = MaterialFunctionCall!(
   Texture=@n2
   UVs=FunctionInput(Preview=Add(A=@n5.UVs, B=@n6)) {bUsePreviewValueAsDefault=true, inputName=MouthUVs, inputType=FunctionInput_Vector2}
   SubImages=Constant() {r=4}
   Frame=ScalarParameter() {parameterName=flipbook_face_mouth_index}
  ) {materialFunction=/Engine/Functions/Engine_MaterialFunctions03/Texturing/SubUV_Function}
  n8 = Constant() {r=0.5}
  n9 = MaterialFunctionCall(Float2=@n7.UVs) {materialFunction=&BreakOutFloat2Components}
  n10 = Multiply(
   A=Power(Base=@n7.Alpha)
   B=Multiply(A=SphereMask(A=@n8, B=@n9.R) {attenuationRadius=0.5}, B=SphereMask(A=@n9.G, B=@n8) {attenuationRadius=0.5})
  )
  n11 = LinearInterpolate(
   A=@n1.EmissiveColor
   B=Multiply(A=@n7.RGB, B=ScalarParameter() {defaultValue=0.3, group=09_Flipbook Mouth, parameterName=MouthEmissiveIntensity})
   Alpha=@n10
  ) {constB=0}
  FunctionOutput(None=SetMaterialAttributes!(
    MaterialAttributes=@n1.MaterialAttributes
    BaseColor=LinearInterpolate(A=@n1.BaseColor, B=@n7.RGB, Alpha=@n10)
    EmissiveColor=StaticSwitchParameter(True=@n11, False=@n1.EmissiveColor) {group=09_Flipbook Mouth, parameterName=useEmissiveforMouth}
   ) {attributeSetTypes [#1, #2]})
  FunctionOutput(None=@n7.RGB) {outputName=Color}
  FunctionOutput(None=@n10) {outputName=Alpha}
 ]}
