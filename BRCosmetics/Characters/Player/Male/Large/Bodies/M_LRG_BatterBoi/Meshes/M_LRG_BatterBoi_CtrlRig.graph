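# Control Rig for the large male BatterBoi player body that applies corrective deformation to limbs and costume bones.
# It runs a shared DeformRig function, then belt, shoe and fix passes, after an optional debug step driving bones from controls.
# Spherical pose readers and DeformRig thigh attributes weight additive offsets on belt, bun, pouch and shoe bones.
# Float variables for large gun, parachute and skydive floating weight arm and tail fixes.
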ControlRigBlueprint /BRCosmetics/Characters/Player/Male/Large/Bodies/M_LRG_BatterBoi/Meshes/M_LRG_BatterBoi_CtrlRig
graphs [
 New Function_ContainedGraph {nodes [
   RigVMFunctionEntryNode()
   > RigVMFunction_Sequence = RigVMAggregateNode()
   n1 = RigVMDispatchNode(BoneName=root, Default=1, Name=deformRig_l_thigh_up_alpha) {template="DISPATCH_RigDispatch_GetAnimAttribute(in Name,in BoneName,in Default,out Value,out Found)"}
   &RigUnit_ModifyTransforms(
    ItemToModify=((Item=(Type=Bone,Name="dyn_belt_front_l"),Transform=(Rotation=(X=0,Y=0,Z=0,W=1),Translation=(X=-8,Y=-2,Z=0),Scale3D=(X=1,Y=1,Z=1))))
    WeightMinimum=0
    Weight=@n1.Value
    ExecutePin=@RigVMFunction_Sequence.A
   )
   > &RigUnit_ModifyTransforms(
    ItemToModify=((Item=(Type=Bone,Name="dyn_belt_side_l"),Transform=(Rotation=(X=0,Y=0,Z=0,W=1),Translation=(X=-3,Y=-3,Z=0),Scale3D=(X=1,Y=1,Z=1))))
    WeightMinimum=0
    Weight=@n1.Value
   )
   > &RigUnit_ModifyTransforms(
    ItemToModify=((Item=(Type=Bone,Name="dyn_bun"),Transform=(Rotation=(X=-0.0136,Y=-0.0266,Z=-0.4545,W=0.8903),Translation=(X=3.604,Y=-1.7351,Z=0),Scale3D=(X=1,Y=1,Z=1))))
    WeightMinimum=0
    Weight=@n1.Value
   )
   RigVMFunctionReturnNode(ExecuteContext=@RigVMFunction_Sequence.E)
   n2 = RigVMDispatchNode(BoneName=root, Default=1, Name=deformRig_r_thigh_up_alpha) {template="DISPATCH_RigDispatch_GetAnimAttribute(in Name,in BoneName,in Default,out Value,out Found)"}
   &RigUnit_ModifyTransforms(
    ItemToModify=((Item=(Type=Bone,Name="dyn_belt_front_r"),Transform=(Rotation=(X=0,Y=0,Z=0,W=1),Translation=(X=8,Y=2,Z=0),Scale3D=(X=1,Y=1,Z=1))))
    WeightMinimum=0
    Weight=@n2.Value
    ExecutePin=@RigVMFunction_Sequence.B
   )
   > &RigUnit_ModifyTransforms(
    ItemToModify=((Item=(Type=Bone,Name="dyn_belt_side_r"),Transform=(Rotation=(X=0,Y=0,Z=0,W=1),Translation=(X=3,Y=3,Z=0),Scale3D=(X=1,Y=1,Z=1))))
    WeightMinimum=0
    Weight=@n2.Value
   )
   n3 = &RigUnit_SphericalPoseReader(
    ActiveRegionScaleFactors=(PositiveWidth=3,NegativeWidth=0.1,PositiveHeight=3,NegativeHeight=3)
    ActiveRegionSize=0.1
    Debug=(bDrawDebug=false,DebugScale=25,DebugSegments=20,DebugThickness=0.25)
    DriverAxis=(X=-1,Y=0,Z=0)
    DriverItem=(Type=Bone,Name="thigh_l")
    FalloffRegionScaleFactors=(PositiveWidth=1,NegativeWidth=1,PositiveHeight=1,NegativeHeight=1)
    FalloffSize=0.45
    OptionalParentItem=(Type=Bone)
    RotationOffset=(X=180,Y=0,Z=0)
    ExecutePin=@RigVMFunction_Sequence.C
   )
   > &RigUnit_ModifyTransforms(
    ItemToModify=((Item=(Type=Bone,Name="dyn_bun"),Transform=(Rotation=(X=0,Y=0,Z=-0.2253,W=0.9743),Translation=(X=0.6856,Y=-2.1684,Z=0),Scale3D=(X=1,Y=1,Z=1))))
    WeightMinimum=0
    Weight=@n3.OutputParam
   )
   > n4 = &RigUnit_SphericalPoseReader(
    ActiveRegionScaleFactors=(PositiveWidth=3,NegativeWidth=0.1,PositiveHeight=3,NegativeHeight=3)
    ActiveRegionSize=0.1
    Debug=(bDrawDebug=false,DebugScale=25,DebugSegments=20,DebugThickness=0.25)
    DriverAxis=(X=1,Y=0,Z=0)
    DriverItem=(Type=Bone,Name="thigh_r")
    FalloffRegionScaleFactors=(PositiveWidth=1,NegativeWidth=1,PositiveHeight=1,NegativeHeight=1)
    FalloffSize=0.45
    OptionalParentItem=(Type=Bone)
    RotationOffset=(X=0,Y=0,Z=180)
   )
   > &RigUnit_ModifyTransforms(
    ItemToModify=((Item=(Type=Bone,Name="base_pouch"),Transform=(Rotation=(X=0,Y=0,Z=-0.8639,W=0.5036),Translation=(X=0,Y=0,Z=0),Scale3D=(X=1,Y=1,Z=1))))
    WeightMinimum=0
    Weight=@n4.OutputParam
   )
   > n5 = &RigUnit_SphericalPoseReader(
    ActiveRegionScaleFactors=(PositiveWidth=3.5,NegativeWidth=0.5,PositiveHeight=3,NegativeHeight=3)
    ActiveRegionSize=0.1
    Debug=(bDrawDebug=false,DebugScale=25,DebugSegments=20,DebugThickness=0.25)
    DriverAxis=(X=1,Y=0,Z=0)
    DriverItem=(Type=Bone,Name="spine_01")
    FalloffRegionScaleFactors=(PositiveWidth=1,NegativeWidth=1,PositiveHeight=1,NegativeHeight=1)
    FalloffSize=0.45
    OptionalParentItem=(Type=Bone)
    RotationOffset=(X=0,Y=175,Z=200)
   )
   > &RigUnit_ModifyTransforms(
    ItemToModify=((Item=(Type=Bone,Name="base_belt_side_l"),Transform=(Rotation=(X=0.0114,Y=-0.0864,Z=0.13,W=0.9877),Translation=(X=4,Y=-4,Z=0),Scale3D=(X=1,Y=1,Z=1))))
    WeightMinimum=0
    Weight=@n5.OutputParam
   )
   > n6 = &RigUnit_SphericalPoseReader(
    ActiveRegionScaleFactors=(PositiveWidth=3.5,NegativeWidth=0.5,PositiveHeight=3,NegativeHeight=3)
    ActiveRegionSize=0.1
    Debug=(bDrawDebug=false,DebugScale=25,DebugSegments=20,DebugThickness=0.25)
    DriverAxis=(X=1,Y=0,Z=0)
    DriverItem=(Type=Bone,Name="spine_01")
    FalloffRegionScaleFactors=(PositiveWidth=1,NegativeWidth=1,PositiveHeight=1,NegativeHeight=1)
    FalloffSize=0.45
    OptionalParentItem=(Type=Bone)
    RotationOffset=(X=180,Y=185,Z=200)
   )
   > &RigUnit_ModifyTransforms(
    ItemToModify=((Item=(Type=Bone,Name="base_belt_side_r"),Transform=(Rotation=(X=0.0114,Y=-0.0864,Z=0.13,W=0.9877),Translation=(X=-4,Y=4,Z=0),Scale3D=(X=1,Y=1,Z=1))))
    WeightMinimum=0
    Weight=@n6.OutputParam
   )
   n7 = &RigUnit_SphericalPoseReader(
    ActiveRegionScaleFactors=(PositiveWidth=3,NegativeWidth=0.5,PositiveHeight=3,NegativeHeight=3)
    ActiveRegionSize=0.1
    Debug=(bDrawDebug=false,DebugScale=25,DebugSegments=20,DebugThickness=0.25)
    DriverAxis=(X=1,Y=0,Z=0)
    DriverItem=(Type=Bone,Name="spine_01")
    FalloffRegionScaleFactors=(PositiveWidth=1,NegativeWidth=1,PositiveHeight=1,NegativeHeight=1)
    FalloffSize=0.45
    OptionalParentItem=(Type=Bone)
    RotationOffset=(X=90,Y=180,Z=200)
    ExecutePin=@RigVMFunction_Sequence.D
   )
   > &RigUnit_ModifyTransforms(
    ItemToModify=((Item=(Type=Bone,Name="dyn_belt_back_l"),Transform=(Rotation=(X=0,Y=0,Z=0,W=1),Translation=(X=3,Y=-4,Z=0),Scale3D=(X=1,Y=1,Z=1))))
    WeightMinimum=0
    Weight=@n7.OutputParam
   )
   > &RigUnit_ModifyTransforms(
    ItemToModify=((Item=(Type=Bone,Name="dyn_belt_back_r"),Transform=(Rotation=(X=0,Y=0,Z=0,W=1),Translation=(X=-3,Y=4,Z=0),Scale3D=(X=1,Y=1,Z=1))))
    WeightMinimum=0
    Weight=@n7.OutputParam
   )
   > &RigUnit_ModifyTransforms(
    ItemToModify=((Item=(Type=Bone,Name="base_tail"),Transform=(Rotation=(X=0,Y=0.0872,Z=0,W=0.9962),Translation=(X=0,Y=0,Z=-3),Scale3D=(X=1,Y=1,Z=1))))
    WeightMinimum=0
    Weight=RigVMVariableNode() {variable=isSkydiveFloating}.Value
   )
   > n8 = &RigUnit_SphericalPoseReader(
    ActiveRegionScaleFactors=(PositiveWidth=1,NegativeWidth=1,PositiveHeight=1,NegativeHeight=1)
    ActiveRegionSize=0.1
    Debug=(bDrawDebug=false,DebugScale=25,DebugSegments=20,DebugThickness=0.25)
    DriverAxis=(X=1,Y=0,Z=0)
    DriverItem=(Type=Bone,Name="spine_01")
    FalloffRegionScaleFactors=(PositiveWidth=1,NegativeWidth=1,PositiveHeight=1,NegativeHeight=1)
    FalloffSize=0.45
    OptionalParentItem=(Type=Bone)
    RotationOffset=(X=-90,Y=180,Z=170)
   )
   > &RigUnit_ModifyTransforms(
    ItemToModify=((Item=(Type=Bone,Name="base_belt_front_l"),Transform=(Rotation=(X=0,Y=0,Z=0,W=1),Translation=(X=0,Y=-15,Z=0),Scale3D=(X=1,Y=1,Z=1))))
    WeightMinimum=0
    Weight=@n8.OutputParam
   )
   > &RigUnit_ModifyTransforms(
    ItemToModify=((Item=(Type=Bone,Name="base_belt_front_r"),Transform=(Rotation=(X=0,Y=0,Z=0,W=1),Translation=(X=0,Y=15,Z=0),Scale3D=(X=1,Y=1,Z=1))))
    WeightMinimum=0
    Weight=@n8.OutputParam
   )
  ]}
 New Function_ContainedGraph {nodes [
   n1 = RigVMFunctionEntryNode()
   RigVMFunction_ControlFlowBranch = &RigVMFunction_ControlFlowBranch(Condition=@n1.Debug, ExecuteContext=@n1.ExecuteContext)
   RigVMFunctionReturnNode(ExecuteContext=@RigVMFunction_ControlFlowBranch.Completed)
   &RigUnit_SetTransform(
    Item=(Type=Bone,Name="spine_01")
    Weight=1
    bInitial=False
    Value=&RigUnit_GetTransform(Item=(Type=Control,Name="spine_01_ctrl"), bInitial=False).Transform
    ExecutePin=@RigVMFunction_ControlFlowBranch.True
   )
   RigUnit_GetTransform = &RigUnit_GetTransform(Item=(Type=Control,Name="thigh_r_ctrl"), bInitial=False)
   > &RigUnit_SetTransform(Item=(Type=Bone,Name="thigh_r"), Weight=1, bInitial=False, Value=@RigUnit_GetTransform.Transform)
   > &RigUnit_SetTransform(
    Item=(Type=Bone,Name="thigh_l")
    Weight=1
    bInitial=False
    bPropagateToChildren=true
    Value=&RigUnit_GetTransform(Item=(Type=Control,Name="thigh_l_ctrl"), bInitial=False).Transform
   )
   > &RigUnit_SetTransform(
    Item=(Type=Bone,Name="foot_l")
    Weight=1
    bInitial=false
    bPropagateToChildren=true
    Value=&RigUnit_GetTransform(Item=(Type=Control,Name="foot_l_ctrl"), bInitial=False).Transform
   )
   > &RigUnit_SetTransform(
    Item=(Type=Bone,Name="foot_r")
    Weight=1
    bInitial=false
    Value=&RigUnit_GetTransform(Item=(Type=Control,Name="foot_r_ctrl"), bInitial=False).Transform
   )
  ]}
 New Function_ContainedGraph {nodes [
   n1 = RigVMFunctionEntryNode()
   RigVMFunction_Sequence = &RigVMFunction_Sequence(ExecuteContext=@n1.ExecuteContext)
   n2 = &RigUnit_SphericalPoseReader(
    ActiveRegionScaleFactors=(PositiveWidth=2,NegativeWidth=0.5,PositiveHeight=3,NegativeHeight=3)
    ActiveRegionSize=0.1
    Debug=(bDrawDebug=false,DebugScale=25,DebugSegments=20,DebugThickness=0.25)
    DriverAxis=(X=0,Y=1,Z=0)
    DriverItem=(Type=Bone,Name="foot_l")
    FalloffRegionScaleFactors=(PositiveWidth=1,NegativeWidth=1,PositiveHeight=1,NegativeHeight=1)
    FalloffSize=0.45
    OptionalParentItem=(Type=Bone)
    RotationOffset=(X=90,Y=180,Z=-90)
    ExecutePin=@RigVMFunction_Sequence.A
   )
   > &RigUnit_ModifyTransforms(
    ItemToModify=((Item=(Type=Bone,Name="base_shoe_l"),Transform=(Rotation=(X=0,Y=-0.5,Z=0,W=0.866),Translation=(X=0,Y=0,Z=0),Scale3D=(X=1,Y=1,Z=1))))
    WeightMinimum=0
    Weight=@n2.OutputParam
   )
   > n3 = &RigUnit_SphericalPoseReader(
    ActiveRegionScaleFactors=(PositiveWidth=2,NegativeWidth=0.5,PositiveHeight=3,NegativeHeight=3)
    ActiveRegionSize=0.1
    Debug=(bDrawDebug=false,DebugScale=25,DebugSegments=20,DebugThickness=0.25)
    DriverAxis=(X=0,Y=-1,Z=0)
    DriverItem=(Type=Bone,Name="foot_r")
    FalloffRegionScaleFactors=(PositiveWidth=1,NegativeWidth=1,PositiveHeight=1,NegativeHeight=1)
    FalloffSize=0.45
    OptionalParentItem=(Type=Bone)
    RotationOffset=(X=-90,Y=0,Z=-90)
   )
   > &RigUnit_ModifyTransforms(
    ItemToModify=((Item=(Type=Bone,Name="base_shoe_r"),Transform=(Rotation=(X=0,Y=-0.5,Z=0,W=0.866),Translation=(X=0,Y=0,Z=0),Scale3D=(X=1,Y=1,Z=1))))
    WeightMinimum=0
    Weight=@n3.OutputParam
   )
   RigVMFunctionReturnNode(ExecuteContext=@RigVMFunction_Sequence.B)
  ]}
 New Function_ContainedGraph {nodes [
   RigVMFunctionEntryNode()
   > RigVMFunction_Sequence = RigVMAggregateNode()
   n1 = &RigUnit_SphericalPoseReader(
    ActiveRegionScaleFactors=(PositiveWidth=0.5,NegativeWidth=1,PositiveHeight=0.5,NegativeHeight=0.5)
    ActiveRegionSize=0.1
    Debug=(bDrawDebug=false,bDraw2D=False,bDrawLocalAxes=False,DebugScale=25,DebugSegments=20,DebugThickness=0.25)
    DriverAxis=(X=-1,Y=0,Z=0)
    DriverItem=(Type=Bone,Name="hand_r")
    FalloffRegionScaleFactors=(PositiveWidth=0.2,NegativeWidth=1,PositiveHeight=0.5,NegativeHeight=0.5)
    FalloffSize=0.2
    FlipHeightScaling=False
    FlipWidthScaling=False
    OptionalParentItem=(Type=Bone,Name="None")
    RotationOffset=(X=140,Y=0,Z=30)
    ExecutePin=@RigVMFunction_Sequence.A
   )
   Multiply = &RigVMFunction_MathDoubleMul(A=@n1.OutputParam, B=RigVMVariableNode() {variable=isLargeGunEquipped}.Value)
   Remap = &RigVMFunction_MathDoubleRemap(SourceMaximum=1, SourceMinimum=0, TargetMaximum=1, TargetMinimum=0, bClamp=False, Value=@Multiply.Result)
   > &RigUnit_ModifyTransforms(
    ItemToModify=((Item=(Type=Bone,Name="lowerarm_twist_02_l"),Transform=(Rotation=(X=0.342,Y=0,Z=0,W=0.9397),Translation=(X=0,Y=0,Z=0),Scale3D=(X=1,Y=1.5,Z=1))))
    WeightMinimum=0
    Weight=@Remap.Result
   )
   n2 = RigVMVariableNode() {variable=isParachuteOpen}
   &RigUnit_ModifyTransforms(
    ItemToModify=((Item=(Type=Bone,Name="deform_upperarm_up_base_l"),Transform=(Rotation=(X=0,Y=0,Z=0,W=1),Translation=(X=10,Y=0,Z=0),Scale3D=(X=1,Y=1,Z=1))))
    WeightMinimum=0
    ExecutePin=@RigVMFunction_Sequence.B
    Weight=@n2.Value
   )
   > &RigUnit_ModifyTransforms(
    ItemToModify=((Item=(Type=Bone,Name="deform_upperarm_up_base_r"),Transform=(Rotation=(X=0,Y=0,Z=0,W=1),Translation=(X=-10,Y=0,Z=0),Scale3D=(X=1,Y=1,Z=1))))
    WeightMinimum=0
    Weight=@n2.Value
   )
   RigVMFunctionReturnNode(ExecuteContext=@RigVMFunction_Sequence.C)
  ]}
 RigVMFunctionLibrary {nodes [
   Fixes = RigVMCollapseNode(ExecuteContext=())
   beltSystem = RigVMCollapseNode(ExecuteContext=())
   ctrlSetup = RigVMCollapseNode(Debug=False, ExecuteContext=())
   shoesSystem = RigVMCollapseNode(ExecuteContext=())
  ]}
 RigVMFunction_Sequence_1_ContainedGraph {nodes [
   n1 = RigVMFunctionEntryNode()
   RigVMFunction_Sequence = &RigVMFunction_Sequence(ExecuteContext=@n1.ExecuteContext)
   n2 = &RigVMFunction_Sequence(ExecuteContext=@RigVMFunction_Sequence.B)
   n3 = &RigVMFunction_Sequence(ExecuteContext=@n2.B)
   n4 = &RigVMFunction_Sequence(ExecuteContext=@n3.B)
   RigVMFunctionReturnNode(A=@RigVMFunction_Sequence.A, B=@n2.A, C=@n3.A, D=@n4.A, E=@n4.B)
  ]}
 RigVMFunction_Sequence_1_ContainedGraph {nodes [
   n1 = RigVMFunctionEntryNode()
   RigVMFunction_Sequence = &RigVMFunction_Sequence(ExecuteContext=@n1.ExecuteContext)
   n2 = &RigVMFunction_Sequence(ExecuteContext=@RigVMFunction_Sequence.B)
   RigVMFunctionReturnNode(A=@RigVMFunction_Sequence.A, B=@n2.A, C=@n2.B)
  ]}
 RigVMModel {nodes [
   RigUnit_BeginExecution = &RigUnit_BeginExecution()
   RigVMFunction_Sequence = &RigVMFunction_Sequence(ExecuteContext=@RigUnit_BeginExecution.ExecutePin)
   ctrlSetup = RigVMFunctionReferenceNode(Debug=false, ExecuteContext=@RigVMFunction_Sequence.A)
   FN_DeformRig_FNC = RigVMFunctionReferenceNode(
    DeformRig_ Curve="(deformRig_l_upperarm_up_alpha_72_#1=False,deformRig_r_upperarm_up_alpha_73_#2=False,deformRig_l_upperarm_fr_alpha_74_#3=False,deformRig_r_upperarm_fr_alpha_75_CB51B8 ...(truncated)"
    DeformRig_Bool="(upperarm_1_#4=True,elbow_3_#5=True,wrist_7_#6=True,groin_8_#7=True,glute_17_#8=True,knee_9_40C408324CA261 ...(truncated)"
    DeformRig_Debug=(upperarm_debug_18_#4=False,wrist_debug_20_#6=False,lat_debug_26_#9=False,groinGlute_debug_27_#7=False)
    DeformRig_Value="(upperarm_up_18_#10=(Rotation=(X=0,Y=0,Z=0,W=1),Translation=(X=0,Y=0,Z=8),Scale3D=(X=0.7,Y=1,Z=1)),upperarm_fwd_fr_19_#11=(Rotation=(X=0,Y=0,Z=0,W=1),Translation=(X=0,Y=10,Z=0),Scale3D=(X=0.5,Y=1,Z= ...(truncated)"
    ExecuteContext=@RigVMFunction_Sequence.B
   )
   > beltSystem = RigVMFunctionReferenceNode()
   > shoesSystem = RigVMFunctionReferenceNode()
   > Fixes = RigVMFunctionReferenceNode()
  ]}
]
hierarchy [
 root {loc=<0 0 0>, rot=<0 0 0 1>}
 pelvis {parent=root, loc=<0 -0.7321 92.4218>, rot=<0.0127 0.707 -0.0127 -0.707>}
 spine_01 {parent=pelvis, loc=<5.6002 -0.5128 0>, rot=<0 0 0.034 -0.9994>}
 spine_02 {parent=spine_01, loc=<8.0756 0.0139 0>, rot=<0 0 0.0238 -0.9997>}
 spine_03 {parent=spine_02, loc=<7.0395 0 0>, rot=<0 0 0.0149 -0.9999>}
 spine_04 {parent=spine_03, loc=<8.1023 0 0>, rot=<0 0 -0.0132 -0.9999>}
 spine_05 {parent=spine_04, loc=<10.0141 0 0>, rot=<0 0 -0.0015 -1>}
 attach_backpack {parent=spine_05, loc=<-0.123 -15.1423 0>, rot=<-0.0093 -0.707 0.0093 -0.707>}
 clavicle_l {parent=spine_05, loc=<2.5856 -0.2866 -4.3777>, rot=<0.0465 -0.7289 -0.009 -0.6829>}
 upperarm_l {parent=clavicle_l, loc=<18.2264 0.1342 -0.001>, rot=<0.0253 -0.2693 0.0013 -0.9627>}
 lowerarm_l {parent=upperarm_l, loc=<29.0071 0.1686 -1.239>, rot=<0.0391 -0.0099 -0.2453 -0.9686>}
 hand_l {parent=lowerarm_l, loc=<25.8607 0.222 1.0274>, rot=<-0.7193 -0.0988 0.1193 -0.6773>}
 thumb_01_l {parent=hand_l, loc=<1.3118 -0.3191 -2.459>, rot=<0.6546 -0.4042 0.0664 -0.6354>}
 thumb_02_l {parent=thumb_01_l, loc=<5.9336 -0.0448 0.0401>, rot=<-0.0037 0.0213 0.298 -0.9543>}
 thumb_03_l {parent=thumb_02_l, loc=<2.2223 0.3682 -0.0907>, rot=<0.0118 -0.013 0.0066 -0.9998>}
 index_metacarpal_l {parent=hand_l, loc=<1.6535 0.0183 -2.96>, rot=<0.0158 -0.0124 -0.0173 -0.9996>}
 index_01_l {parent=index_metacarpal_l, loc=<8.3008 0.0357 -0.0524>, rot=<-0.0044 0.0485 0.0924 -0.9945>}
 index_02_l {parent=index_01_l, loc=<3.3587 -0.3004 0.0452>, rot=<-0.002 0.0027 0.3225 -0.9466>}
 index_03_l {parent=index_02_l, loc=<2.5578 -0.123 -0.0252>, rot=<-0.002 0.0067 0.2558 -0.9667>}
 middle_metacarpal_l {parent=hand_l, loc=<2.1856 -0.1725 -0.5984>, rot=<0.006 0.0168 -0.0329 -0.9993>}
 middle_01_l {parent=middle_metacarpal_l, loc=<7.3275 -0.2447 0.0129>, rot=<-0.0208 0.0042 0.0858 -0.9961>}
 middle_02_l {parent=middle_01_l, loc=<4.0123 0.1401 0.1175>, rot=<-0.0218 0.0258 0.3512 -0.9357>}
 middle_03_l {parent=middle_02_l, loc=<2.6077 0.0534 0.1373>, rot=<0.0082 0.0188 0.1708 -0.9851>}
 ring_metacarpal_l {parent=hand_l, loc=<2.2074 -0.3625 1.3638>, rot=<-0.0284 0.0722 -0.0458 -0.9959>}
 ring_01_l {parent=ring_metacarpal_l, loc=<7.0295 -0.6197 -0.1446>, rot=<-0.0489 -0.0252 0.1892 -0.9804>}
 ring_02_l {parent=ring_01_l, loc=<3.6174 0.4647 -0.1027>, rot=<0 0 0.2806 -0.9598>}
 ring_03_l {parent=ring_02_l, loc=<2.4152 0.114 0.0199>, rot=<0 0 0.1189 -0.9929>}
 pinky_metacarpal_l {parent=hand_l, loc=<1.7616 -0.9753 3.3814>, rot=<-0.0397 0.0814 -0.069 -0.9935>}
 pinky_01_l {parent=pinky_metacarpal_l, loc=<7.1709 -0.4831 -0.1094>, rot=<-0.0654 -0.0035 0.2242 -0.9723>}
 pinky_02_l {parent=pinky_01_l, loc=<2.8274 0.0893 -0.1043>, rot=<0.0037 0.0011 0.2854 -0.9584>}
 pinky_03_l {parent=pinky_02_l, loc=<2.2046 0.0307 -0.0078>, rot=<-0.0002 0.0098 0.1445 -0.9895>}
 deform_wrist_dn_base_l {parent=hand_l, loc=<0 -3.4173 0>, rot=<0 0 0 -1>}
 deform_wrist_dn_l {parent=deform_wrist_dn_base_l, loc=<0 0 0>, rot=<0 0 0 1>}
 deform_wrist_up_base_l {parent=hand_l, loc=<0 2.7805 0>, rot=<0 0 0 -1>}
 deform_wrist_up_l {parent=deform_wrist_up_base_l, loc=<0 0 0>, rot=<0 0 0 1>}
 base_endGloves_l {parent=hand_l, loc=<0.1513 -1.9396 5.1119>, rot=<0.5712 0.2409 0.4141 -0.6665>}
 dyn_endGloves_l {parent=base_endGloves_l, loc=<0 0 0>, rot=<0 0 0 1>}
 weapon_l {parent=hand_l, loc=<-0.1974 -3.4111 0.3811>, rot=<0 0 0 1>}
 hand_attach_l {parent=hand_l, loc=<8.5 -3 0>, rot=<0.7071 0.7071 0 0>}
 lowerarm_twist_02_l {parent=lowerarm_l, loc=<9.6199 0 0>, rot=<0 0 0 -1>}
 lowerarm_twist_01_l {parent=lowerarm_l, loc=<17.8549 0 0>, rot=<0 0 0 -1>}
 deform_elbow_in_base_l {parent=lowerarm_l, loc=<0.2112 5.645 0.8274>, rot=<0 0 0 -1>}
 deform_elbow_in_l {parent=deform_elbow_in_base_l, loc=<0 0 0>, rot=<0 0 0 1>}
 deform_elbow_out_base_l {parent=lowerarm_l, loc=<0 -7.5129 0>, rot=<0 0 0 -1>}
 deform_elbow_out_l {parent=deform_elbow_out_base_l, loc=<0 0 0>, rot=<0 0 0 1>}
 upperarm_twist_01_l {parent=upperarm_l, loc=<7.9796 0.0843 -0.6195>, rot=<0 0 0 -1>}
 deform_upperarm_bk_base_l {parent=upperarm_twist_01_l, loc=<-6.8983 -13.3826 -1.3258>, rot=<0 0 0 1>}
 deform_upperarm_bk_l {parent=deform_upperarm_bk_base_l, loc=<0 0 0>, rot=<0 0 0 1>}
 deform_upperarm_fr_base_l {parent=upperarm_twist_01_l, loc=<-8.3487 10.7117 -2.1455>, rot=<0 0 0 1>}
 deform_upperarm_fr_l {parent=deform_upperarm_fr_base_l, loc=<0 0 0>, rot=<0 0 0 1>}
 deform_upperarm_up_base_l {parent=upperarm_twist_01_l, loc=<-11.2012 -1.2234 8.7784>, rot=<0 0 0 1>}
 deform_upperarm_up_l {parent=deform_upperarm_up_base_l, loc=<0 0 0>, rot=<0 0 0 1>}
 upperarm_twist_02_l {parent=upperarm_l, loc=<19.5059 0.0843 -0.6195>, rot=<0 0 0 -1>}
 deform_bicep_base_l {parent=upperarm_twist_02_l, loc=<0 8.0494 0>, rot=<0 0 0 1>}
 deform_bicep_l {parent=deform_bicep_base_l, loc=<0 0 0>, rot=<0 0 0 1>}
 neck_01 {parent=spine_05, loc=<10.815 1.9196 0>, rot=<0 0 -0.1864 -0.9825>}
 neck_02 {parent=neck_01, loc=<4.7347 0.3045 0>, rot=<0 0 0.0363 -0.9993>}
 head {parent=neck_02, loc=<5.7632 0.5552 0>, rot=<0 0 0.115 -0.9934>}
 attach_cape {parent=spine_05, loc=<9.1509 -13.3783 0>, rot=<0.0616 -0.7044 -0.0616 -0.7044>}
 clavicle_r {parent=spine_05, loc=<2.5855 -0.2866 4.3777>, rot=<0.7289 0.0465 -0.6829 0.009>}
 upperarm_r {parent=clavicle_r, loc=<-18.2264 -0.1342 0.001>, rot=<0.0253 -0.2693 0.0013 -0.9627>}
 lowerarm_r {parent=upperarm_r, loc=<-29.0071 -0.1686 1.239>, rot=<0.0391 -0.0099 -0.2453 -0.9686>}
 hand_r {parent=lowerarm_r, loc=<-25.8607 -0.222 -1.0274>, rot=<-0.7193 -0.0988 0.1193 -0.6773>}
 thumb_01_r {parent=hand_r, loc=<-1.3118 0.3191 2.459>, rot=<0.6546 -0.4042 0.0664 -0.6354>}
 thumb_02_r {parent=thumb_01_r, loc=<-5.9336 0.0448 -0.0401>, rot=<-0.0037 0.0213 0.298 -0.9543>}
 thumb_03_r {parent=thumb_02_r, loc=<-2.2223 -0.3682 0.0907>, rot=<0.0118 -0.013 0.0066 -0.9998>}
 index_metacarpal_r {parent=hand_r, loc=<-1.6535 -0.0183 2.96>, rot=<0.0158 -0.0124 -0.0173 -0.9996>}
 index_01_r {parent=index_metacarpal_r, loc=<-8.3008 -0.0357 0.0524>, rot=<-0.0044 0.0485 0.0924 -0.9945>}
 index_02_r {parent=index_01_r, loc=<-3.3587 0.3004 -0.0452>, rot=<-0.002 0.0027 0.3225 -0.9466>}
 index_03_r {parent=index_02_r, loc=<-2.5578 0.123 0.0252>, rot=<-0.002 0.0067 0.2558 -0.9667>}
 middle_metacarpal_r {parent=hand_r, loc=<-2.1856 0.1725 0.5984>, rot=<0.006 0.0168 -0.0329 -0.9993>}
 middle_01_r {parent=middle_metacarpal_r, loc=<-7.3275 0.2447 -0.0129>, rot=<-0.0208 0.0042 0.0858 -0.9961>}
 middle_02_r {parent=middle_01_r, loc=<-4.0123 -0.1401 -0.1175>, rot=<-0.0218 0.0258 0.3512 -0.9357>}
 middle_03_r {parent=middle_02_r, loc=<-2.6077 -0.0534 -0.1373>, rot=<0.0082 0.0188 0.1708 -0.9851>}
 ring_metacarpal_r {parent=hand_r, loc=<-2.2074 0.3625 -1.3638>, rot=<-0.0284 0.0722 -0.0458 -0.9959>}
 ring_01_r {parent=ring_metacarpal_r, loc=<-7.0295 0.6197 0.1446>, rot=<-0.0489 -0.0252 0.1892 -0.9804>}
 ring_02_r {parent=ring_01_r, loc=<-3.6174 -0.4647 0.1027>, rot=<0 0 0.2806 -0.9598>}
 ring_03_r {parent=ring_02_r, loc=<-2.4152 -0.114 -0.0199>, rot=<0 0 0.1189 -0.9929>}
 pinky_metacarpal_r {parent=hand_r, loc=<-1.7616 0.9753 -3.3814>, rot=<-0.0397 0.0814 -0.069 -0.9935>}
 pinky_01_r {parent=pinky_metacarpal_r, loc=<-7.1709 0.4831 0.1094>, rot=<-0.0654 -0.0035 0.2242 -0.9723>}
 pinky_02_r {parent=pinky_01_r, loc=<-2.8274 -0.0893 0.1043>, rot=<0.0037 0.0011 0.2854 -0.9584>}
 pinky_03_r {parent=pinky_02_r, loc=<-2.2046 -0.0307 0.0078>, rot=<-0.0002 0.0098 0.1445 -0.9895>}
 deform_wrist_dn_base_r {parent=hand_r, loc=<0 3.4172 0>, rot=<0 0 0 -1>}
 deform_wrist_dn_r {parent=deform_wrist_dn_base_r, loc=<0 0 0>, rot=<0 0 0 -1>}
 deform_wrist_up_base_r {parent=hand_r, loc=<-0.0001 -2.7804 0>, rot=<0 0 0 -1>}
 deform_wrist_up_r {parent=deform_wrist_up_base_r, loc=<0 0 0>, rot=<0 0 0 -1>}
 base_endGloves_r {parent=hand_r, loc=<-0.1513 1.9396 -5.112>, rot=<0.5712 0.2409 0.4141 -0.6665>}
 dyn_endGloves_r {parent=base_endGloves_r, loc=<0 0 0>, rot=<0 0 0 1>}
 hand_attach_r {parent=hand_r, loc=<-8.5 3 0>, rot=<0 0 -0.7071 -0.7071>}
 weapon_r {parent=hand_r, loc=<0.1974 3.4111 -0.3811>, rot=<0 0 0 1>}
 lowerarm_twist_02_r {parent=lowerarm_r, loc=<-9.6199 0 0>, rot=<0 0 0 -1>}
 lowerarm_twist_01_r {parent=lowerarm_r, loc=<-17.8549 0 0>, rot=<0 0 0 -1>}
 deform_elbow_in_base_r {parent=lowerarm_r, loc=<-0.2114 -5.6449 -0.827>, rot=<0 0 0 -1>}
 deform_elbow_in_r {parent=deform_elbow_in_base_r, loc=<0 0 0>, rot=<0 0 0 1>}
 deform_elbow_out_base_r {parent=lowerarm_r, loc=<-0.0001 7.5129 0.0002>, rot=<0 0 0 -1>}
 deform_elbow_out_r {parent=deform_elbow_out_base_r, loc=<0 0 0>, rot=<0 0 0 1>}
 upperarm_twist_01_r {parent=upperarm_r, loc=<-7.9796 -0.0843 0.6195>, rot=<0 0 0 -1>}
 deform_upperarm_bk_base_r {parent=upperarm_twist_01_r, loc=<6.8985 13.3826 1.3257>, rot=<0 0 0 1>}
 deform_upperarm_bk_r {parent=deform_upperarm_bk_base_r, loc=<0 0 0>, rot=<0 0 0 1>}
 deform_upperarm_fr_base_r {parent=upperarm_twist_01_r, loc=<8.3485 -10.7117 2.1456>, rot=<0 0 0 1>}
 deform_upperarm_fr_r {parent=deform_upperarm_fr_base_r, loc=<0 0 0>, rot=<0 0 0 1>}
 deform_upperarm_up_base_r {parent=upperarm_twist_01_r, loc=<11.2015 1.2234 -8.7788>, rot=<0 0 0 1>}
 deform_upperarm_up_r {parent=deform_upperarm_up_base_r, loc=<0 0 0>, rot=<0 0 0 1>}
 upperarm_twist_02_r {parent=upperarm_r, loc=<-19.5059 -0.0843 0.6195>, rot=<0 0 0 -1>}
 deform_bicep_base_r {parent=upperarm_twist_02_r, loc=<0.0002 -8.0494 -0.0003>, rot=<0 0 0 1>}
 deform_bicep_r {parent=deform_bicep_base_r, loc=<0 0 0>, rot=<0 0 0 1>}
 base_chest {parent=spine_05, loc=<0.946 8.8774 0>, rot=<0.0284 -0.7065 -0.0284 -0.7065>}
 base_chest_l {parent=spine_05, loc=<0.4562 11.5645 -4.6862>, rot=<0.0284 -0.7065 -0.0284 -0.7065>}
 base_chest_r {parent=spine_05, loc=<0.4562 11.5645 4.6862>, rot=<0.0284 -0.7065 -0.0284 -0.7065>}
 base_necklace {parent=spine_05, loc=<9.9021 3.7134 0>, rot=<0 0 0.9302 0.3671>}
 dyn_necklace_01 {parent=base_necklace, loc=<0 0 0>, rot=<0 0 0 1>}
 dyn_necklace_02 {parent=dyn_necklace_01, loc=<6.8842 0 0>, rot=<0 0 0.0883 -0.9961>}
 dyn_necklace_03 {parent=dyn_necklace_02, loc=<1.3556 0 0>, rot=<0 0 -0.095 -0.9955>}
 deform_pec_aim_base_l {parent=spine_04, loc=<5.0227 13.6719 -2.6508>, rot=<0.0672 -0.5201 0.3228 -0.7879>}
 deform_pec_aim_l {parent=deform_pec_aim_base_l, loc=<0 0 0>, rot=<0 0 0 1>}
 deform_pec_base_l {parent=deform_pec_aim_l, loc=<16.3279 0 0>, rot=<0 0 0 1>}
 deform_pec_l {parent=deform_pec_base_l, loc=<0 0 0>, rot=<0 0 0 1>}
 deform_pec_aim_base_r {parent=spine_04, loc=<5.0225 13.6718 2.6508>, rot=<-0.5201 -0.0672 0.7879 0.3228>}
 deform_pec_aim_r {parent=deform_pec_aim_base_r, loc=<0 0 0>, rot=<0 0 0 1>}
 deform_pec_base_r {parent=deform_pec_aim_r, loc=<-16.3279 0 0>, rot=<0 0 0 -1>}
 deform_pec_r {parent=deform_pec_base_r, loc=<0 0 0.0002>, rot=<0 0 0 -1>}
 deform_lat_l_base {parent=spine_03, loc=<2.5561 -10.3716 -17.6918>, rot=<0 0 0 -1>}
 deform_lat_l {parent=deform_lat_l_base, loc=<0 0 0>, rot=<0 0 0 1>}
 deform_lat_r_base {parent=spine_03, loc=<2.5562 -10.3716 17.6918>, rot=<0 0 1 0>}
 deform_lat_r {parent=deform_lat_r_base, loc=<0 0 0>, rot=<0 0 0 -1>}
 thigh_l {parent=pelvis, loc=<-1.8716 -0.0109 -11.9936>, rot=<0.0012 0.0572 0.0155 -0.9982>}
 calf_l {parent=thigh_l, loc=<-41.9178 -0.2246 0.0926>, rot=<0.0006 -0.0161 -0.0391 -0.9991>}
 foot_l {parent=calf_l, loc=<-38.7057 1.2438 0.2452>, rot=<0.0975 -0.0474 0.0317 -0.9936>}
 ball_l {parent=foot_l, loc=<-6.1614 13.1469 0.6807>, rot=<0.0001 -0.0001 -0.7063 -0.7079>}
 base_shoe_l {parent=foot_l, loc=<-1.8262 5.4449 1.4411>, rot=<0.6929 0.141 -0.1585 -0.6891>}
 dyn_shoe_l {parent=base_shoe_l, loc=<0 0 0>, rot=<0 0 0 1>}
 calf_twist_02_l {parent=calf_l, loc=<-15.9268 -0.0967 0.0397>, rot=<0.0237 0.0095 -0.0035 -0.9997>}
 calf_twist_01_l {parent=calf_l, loc=<-32.9335 -0.0967 0.0397>, rot=<0.0237 0.0095 -0.0035 -0.9997>}
 deform_knee_in_base_l {parent=calf_l, loc=<1.298 -8.0356 -0.0496>, rot=<0 0 0 -1>}
 deform_knee_in_l {parent=deform_knee_in_base_l, loc=<0 0 0>, rot=<0 0 0 1>}
 deform_knee_out_base_l {parent=calf_l, loc=<0 6.8739 0>, rot=<0 0 0 -1>}
 deform_knee_out_l {parent=deform_knee_out_base_l, loc=<0 0 0>, rot=<0 0 0 1>}
 thigh_twist_01_l {parent=thigh_l, loc=<-21.6303 0 0>, rot=<0.0013 0 0 -1>}
 deform_groin_base_l {parent=thigh_l, loc=<-2.1908 8.5661 -1.7442>, rot=<0 0 0 -1>}
 deform_groin_l {parent=deform_groin_base_l, loc=<0 0 0>, rot=<0 0 0 1>}
 thigh_r {parent=pelvis, loc=<-1.8716 -0.0109 11.9936>, rot=<0.0572 -0.0012 0.9982 0.0155>}
 calf_r {parent=thigh_r, loc=<41.9178 0.2246 -0.0926>, rot=<0.0006 -0.0161 -0.0391 -0.9991>}
 foot_r {parent=calf_r, loc=<38.7057 -1.2438 -0.2452>, rot=<0.0975 -0.0474 0.0317 -0.9936>}
 ball_r {parent=foot_r, loc=<6.1614 -13.1469 -0.6807>, rot=<0.0001 -0.0001 -0.7063 -0.7079>}
 base_shoe_r {parent=foot_r, loc=<1.8262 -5.4449 -1.4412>, rot=<0.6929 0.141 -0.1585 -0.6891>}
 dyn_shoe_r {parent=base_shoe_r, loc=<0 0 0>, rot=<0 0 0 1>}
 calf_twist_02_r {parent=calf_r, loc=<15.9268 0.0967 -0.0397>, rot=<0.0237 0.0095 -0.0035 -0.9997>}
 calf_twist_01_r {parent=calf_r, loc=<32.9335 0.0967 -0.0397>, rot=<0.0237 0.0095 -0.0035 -0.9997>}
 deform_knee_in_base_r {parent=calf_r, loc=<-1.2981 8.0356 0.0496>, rot=<0 0 0 -1>}
 deform_knee_in_r {parent=deform_knee_in_base_r, loc=<0 0 0>, rot=<0 0 0 1>}
 deform_knee_out_base_r {parent=calf_r, loc=<0 -6.8739 0>, rot=<0 0 0 -1>}
 deform_knee_out_r {parent=deform_knee_out_base_r, loc=<0 0 0>, rot=<0 0 0 1>}
 thigh_twist_01_r {parent=thigh_r, loc=<21.6303 0 0>, rot=<0.0013 0 0 -1>}
 deform_groin_base_r {parent=thigh_r, loc=<2.1908 -8.5661 1.7442>, rot=<0 0 0 -1>}
 deform_groin_r {parent=deform_groin_base_r, loc=<0 0 0>, rot=<0 0 0 -1>}
 deform_glute_base_l {parent=pelvis, loc=<-10.1877 -11.6217 -12.899>, rot=<0.0275 0.0501 -0.4472 -0.8926>}
 deform_glute_l {parent=deform_glute_base_l, loc=<0 0 0>, rot=<0 0 0 1>}
 deform_glute_base_r {parent=pelvis, loc=<-10.1877 -11.6218 12.899>, rot=<0.0501 -0.0275 0.8926 -0.4472>}
 deform_glute_r {parent=deform_glute_base_r, loc=<0 0 0>, rot=<0 0 0 1>}
 base_pouch {parent=pelvis, loc=<0.5817 1.2952 23.0338>, rot=<-0.1167 -0.6974 0.6974 -0.1167>}
 dyn_pouch {parent=base_pouch, loc=<0 0 0>, rot=<0 0 0 1>}
 dyn_zipper {parent=dyn_pouch, loc=<2.738 -3.4181 7.2269>, rot=<0.0069 0.1557 0.0662 -0.9856>}
 base_tail {parent=pelvis, loc=<-3.0002 -12.7031 -0.3332>, rot=<0.537 -0.4624 0.4747 -0.522>}
 dyn_tail_01 {parent=base_tail, loc=<0 0 0>, rot=<0 0 0 1>}
 base_belt_front_l {parent=pelvis, loc=<1.6245 9.3644 -12.9506>, rot=<0 0.2588 0.9659 0>}
 dyn_belt_front_l {parent=base_belt_front_l, loc=<0 0 0>, rot=<0 0 0 1>}
 dyn_bun {parent=dyn_belt_front_l, loc=<0 0 0>, rot=<-0.0433 0.0049 0.1131 -0.9926>}
 dyn_top_bun_l {parent=dyn_bun, loc=<-0.7339 -3.0464 -2.2891>, rot=<0.1572 -0.0714 -0.0672 -0.9827>}
 dyn_top_bun_r {parent=dyn_bun, loc=<-0.1783 -2.5443 1.9022>, rot=<-0.1246 0.0692 -0.0172 -0.9896>}
 dyn_bun_01_l {parent=dyn_bun, loc=<3.736 -2.2604 -2.3395>, rot=<-0.1348 -0.1517 -0.0122 -0.9791>}
 dyn_bun_02_l {parent=dyn_bun_01_l, loc=<4.1461 0 0>, rot=<0 0 0 1>}
 dyn_bun_03_l {parent=dyn_bun_02_l, loc=<4.1461 0 0>, rot=<0 0 0 1>}
 dyn_bun_01_r {parent=dyn_bun, loc=<4.1872 -1.6472 2.2118>, rot=<0.1341 -0.0148 -0.0264 -0.9905>}
 dyn_bun_02_r {parent=dyn_bun_01_r, loc=<4.146 0 0>, rot=<0 0 0 1>}
 dyn_bun_03_r {parent=dyn_bun_02_r, loc=<4.146 0 0>, rot=<0 0 0 1>}
 base_belt_side_l {parent=pelvis, loc=<4.0501 -2.6383 -18.061>, rot=<0 0.7071 0.7071 0>}
 dyn_belt_side_l {parent=base_belt_side_l, loc=<0 0 0>, rot=<0 0 0 1>}
 base_belt_back_l {parent=pelvis, loc=<6.6814 -11.3071 -11.8041>, rot=<0 -0.9659 -0.2588 0>}
 dyn_belt_back_l {parent=base_belt_back_l, loc=<0 0 0>, rot=<0 0 0 1>}
 base_belt_front_r {parent=pelvis, loc=<1.6245 9.3644 12.9506>, rot=<-0.2588 0 0 -0.9659>}
 dyn_belt_front_r {parent=base_belt_front_r, loc=<0 0 0>, rot=<0 0 0 1>}
 base_belt_side_r {parent=pelvis, loc=<4.0501 -2.6383 18.061>, rot=<-0.7071 0 0 -0.7071>}
 dyn_belt_side_r {parent=base_belt_side_r, loc=<0 0 0>, rot=<0 0 0 1>}
 base_belt_back_r {parent=pelvis, loc=<6.6814 -11.3071 11.8041>, rot=<0.9659 0 0 0.2588>}
 dyn_belt_back_r {parent=base_belt_back_r, loc=<0 0 0>, rot=<0 0 0 1>}
 spine_01_ctrl {controlType=EULER_TRANSFORM, type=Control, loc=<0 0 0>, rot=<0 0 0 1>}
 Hide Element 01 {type=Curve, loc=<0 0 0>, rot=<0 0 0 1>}
 thigh_l_ctrl {controlType=EULER_TRANSFORM, type=Control, loc=<0 0 0>, rot=<0 0 0 1>}
 foot_l_ctrl {controlType=EULER_TRANSFORM, parent=thigh_l_ctrl, type=Control, loc=<0 0 0>, rot=<0 0 0 1>}
 thigh_r_ctrl {controlType=EULER_TRANSFORM, type=Control, loc=<0 0 0>, rot=<0 0 0 1>}
 foot_r_ctrl {controlType=EULER_TRANSFORM, parent=thigh_r_ctrl, type=Control, loc=<0 0 0>, rot=<0 0 0 1>}
]
variables [
 isLargeGunEquipped {type=double, default=0}
 isParachuteOpen {type=double, default=0}
 isSkydiveFloating {type=double, default=0}
]
