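# Physics asset for the necklace mesh of the large male BatterBoi player body, defining collision and joints for a dangling necklace chain.
# Three dyn_necklace capsule bodies form a chain hanging from spine_05, with narrow swing limits and locked twist at each joint.
# Capsules on the center, left and right chest bones give the necklace collision volumes, each constrained to spine_05 with free angular motion.
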PhysicsAsset /BRCosmetics/Characters/Player/Male/Large/Bodies/M_LRG_BatterBoi/Meshes/M_LRG_BatterBoi_Necklace_Physics
bodies {
 base_chest [{shapeType=Capsule, length=13.0843, radius=22.6786, shapeName=base_chest_capsule, center=<0 -18.0011 -12.4359>, rotation=<rot 90 0 0>}]
 base_chest_l [{
   shapeType=Capsule
   length=8.7629
   radius=6.709
   shapeName=base_chest_l_capsule
   center=<3.3103 -6.5092 -2.8871>
   rotation=<rot -70 0.0009 -22.9614>
  }]
 base_chest_r [{
   shapeType=Capsule
   length=8.7629
   radius=6.709
   shapeName=base_chest_r_capsule
   center=<-3.3103 -6.5092 -2.8871>
   rotation=<rot 70 -0.0009 -22.9614>
  }]
 dyn_necklace_01 [{shapeType=Capsule, length=4, radius=2, shapeName=dyn_necklace_01_capsule, center=<2 -0.9272 0>, rotation=<rot 90 0 0>}]
 dyn_necklace_02 [{shapeType=Capsule, length=4, radius=2, shapeName=dyn_necklace_03_capsule, center=<0.4091 -1.0262 0>, rotation=<rot 90 178.2101 169.9911>}]
 dyn_necklace_03 [
  {shapeType=Capsule, length=4, radius=2, shapeName=dyn_necklace_03_capsule, center=<3.4625 -1.2349 0.1666>, rotation=<rot 90 180 -179.371>}
  {shapeType=Capsule, length=9.2534, radius=1.3085, shapeName=dyn_necklace_03_capsule, center=<3.3093 -0.6394 0>}
 ]
 spine_05 [{shapeType=Capsule, length=2, radius=1, shapeName=spine_05_capsule, center=<2 0 0>, rotation=<rot 90 0 0>}]
}
constraints [
 {bone1Name=dyn_necklace_02, bone2Name=dyn_necklace_01, swing1LimitDegrees=20, swing2LimitDegrees=14, twistLimitDegrees=1, twistMotion=Locked}
 {bone1Name=dyn_necklace_03, bone2Name=dyn_necklace_02, swing1LimitDegrees=20, swing2LimitDegrees=15, twistLimitDegrees=1, twistMotion=Locked}
 {
  bone1Name=base_chest
  bone2Name=spine_05
  swing1LimitDegrees=45
  swing1Motion=Free
  swing2LimitDegrees=45
  swing2Motion=Free
  twistLimitDegrees=45
  twistMotion=Free
 }
 {
  bone1Name=base_chest_l
  bone2Name=spine_05
  swing1LimitDegrees=45
  swing1Motion=Free
  swing2LimitDegrees=45
  swing2Motion=Free
  twistLimitDegrees=45
  twistMotion=Free
 }
 {
  bone1Name=base_chest_r
  bone2Name=spine_05
  swing1LimitDegrees=45
  swing1Motion=Free
  swing2LimitDegrees=45
  swing2Motion=Free
  twistLimitDegrees=45
  twistMotion=Free
 }
 {bone1Name=dyn_necklace_01, bone2Name=spine_05, swing1LimitDegrees=20, swing2LimitDegrees=15, twistLimitDegrees=1, twistMotion=Locked}
]
