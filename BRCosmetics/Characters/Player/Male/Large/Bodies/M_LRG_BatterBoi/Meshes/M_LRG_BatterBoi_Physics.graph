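# Physics asset for the BatterBoi large male player body cosmetic, covering dynamic bun, pouch and zipper bones plus both thighs.
# Capsule bodies form two three-link bun chains and two top-bun bodies on dyn_bun, while the zipper hangs from the pouch, which attaches to base_pouch.
# Constraints set tight twist limits and per-part swing ranges, widest on pouch and zipper.
# The thigh capsules have no constraints and differ in size between left and right.
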PhysicsAsset /BRCosmetics/Characters/Player/Male/Large/Bodies/M_LRG_BatterBoi/Meshes/M_LRG_BatterBoi_Physics
bodies {
 base_pouch [{shapeType=Capsule, length=2, radius=1, shapeName=base_pouch_capsule, center=<2 0 0>, rotation=<rot 90 0 0>}]
 dyn_bun [{shapeType=Capsule, length=2, radius=1, shapeName=dyn_bun_capsule, center=<2 0 0>, rotation=<rot 90 0 0>}]
 dyn_bun_01_l [{shapeType=Capsule, length=0.6, radius=2.25, shapeName=dyn_bun_01_l_capsule, center=<2 0 0>, rotation=<rot 90 0 0>}]
 dyn_bun_01_r [{shapeType=Capsule, length=0.6, radius=2.25, shapeName=dyn_bun_01_r_capsule, center=<2 0 0>, rotation=<rot 90 0 0>}]
 dyn_bun_02_l [{shapeType=Capsule, length=0.6, radius=2.25, shapeName=dyn_bun_02_l_capsule, center=<2 0 0>, rotation=<rot 90 0 0>}]
 dyn_bun_02_r [{shapeType=Capsule, length=0.6, radius=2.25, shapeName=dyn_bun_02_r_capsule, center=<2 0 0>, rotation=<rot 90 0 0>}]
 dyn_bun_03_l [{shapeType=Capsule, length=0.0725, radius=3.4696, shapeName=dyn_bun_03_l_capsule, center=<2 0 0>, rotation=<rot 90 0 0>}]
 dyn_bun_03_r [{shapeType=Capsule, length=0.0725, radius=3.4696, shapeName=dyn_bun_03_r_capsule, center=<2 0 0>, rotation=<rot 90 0 0>}]
 dyn_pouch [
  {
   shapeType=Capsule
   length=7.0259
   radius=3.4951
   shapeName=dyn_pouch_capsule
   center=<5.982 -4.0191 0.5568>
   rotation=<rot -1.0374 -99.6242 -96.8564>
  }
  {
   shapeType=Capsule
   length=6.2895
   radius=4.2316
   shapeName=dyn_pouch_capsule
   center=<8.9986 -5.2849 0.5535>
   rotation=<rot -1.0374 -99.6242 171.5192>
  }
 ]
 dyn_top_bun_l [{
   shapeType=Capsule
   length=2.9397
   radius=5.1528
   shapeName=dyn_top_bun_l_capsule
   center=<-0.3111 -2.07 -2.7872>
   rotation=<rot -3.7078 -0.0001 0>
  }]
 dyn_top_bun_r [{
   shapeType=Capsule
   length=3.4285
   radius=5.5823
   shapeName=dyn_top_bun_r_capsule
   center=<0.1536 -2.3136 3.5552>
   rotation=<rot 4.7035 -0.2228 -10.7765>
  }]
 dyn_zipper [{shapeType=Capsule, length=2, radius=1, shapeName=dyn_zipper_capsule, center=<2 0 0>, rotation=<rot 90 0 0>}]
 thigh_l [{
   shapeType=Capsule
   length=34.063
   radius=12.3701
   shapeName=thigh_l_capsule
   center=<-18.6453 -3.8163 2.0026>
   rotation=<rot 90 180 -178.8882>
  }]
 thigh_r [{
   shapeType=Capsule
   length=15.7815
   radius=10.2853
   shapeName=thigh_l_capsule
   center=<14.0389 0.9091 -2.0026>
   rotation=<rot 87.5924 175.4244 179.9997>
  }]
}
constraints [
 {bone1Name=dyn_bun_02_l, bone2Name=dyn_bun_01_l, swing1LimitDegrees=15, swing2LimitDegrees=15, twistLimitDegrees=1}
 {bone1Name=dyn_bun_03_l, bone2Name=dyn_bun_02_l, swing1LimitDegrees=15, swing2LimitDegrees=15, twistLimitDegrees=1}
 {bone1Name=dyn_bun_02_r, bone2Name=dyn_bun_01_r, swing1LimitDegrees=15, swing2LimitDegrees=15, twistLimitDegrees=1}
 {bone1Name=dyn_bun_03_r, bone2Name=dyn_bun_02_r, swing1LimitDegrees=15, swing2LimitDegrees=15, twistLimitDegrees=1}
 {bone1Name=dyn_pouch, bone2Name=base_pouch, swing1LimitDegrees=45, swing2LimitDegrees=5, twistLimitDegrees=0}
 {bone1Name=dyn_top_bun_l, bone2Name=dyn_bun, swing1LimitDegrees=5, swing2LimitDegrees=25, twistLimitDegrees=1}
 {bone1Name=dyn_top_bun_r, bone2Name=dyn_bun, swing1LimitDegrees=5, swing2LimitDegrees=25, twistLimitDegrees=1}
 {bone1Name=dyn_bun_01_l, bone2Name=dyn_bun, swing1LimitDegrees=15, swing2LimitDegrees=15, twistLimitDegrees=1}
 {bone1Name=dyn_bun_01_r, bone2Name=dyn_bun, swing1LimitDegrees=15, swing2LimitDegrees=15, twistLimitDegrees=1}
 {bone1Name=dyn_zipper, bone2Name=dyn_pouch, swing1LimitDegrees=5, swing2LimitDegrees=50, twistLimitDegrees=1}
]
