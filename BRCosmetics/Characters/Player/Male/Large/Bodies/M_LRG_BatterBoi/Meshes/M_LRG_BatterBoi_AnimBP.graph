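# Secondary-motion AnimBP for the BatterBoi large male player cosmetic, layered on the parent mesh's copied pose.
# A Control Rig, fed skydive, parachute and assault-rifle states, seeds weighted constraints with RigidBody simulation, or crouch-adjusted AnimDynamics buns as fallback, selected by bRigidBodyEnabled.
# AnimDynamics then moves glove ends and tail on either branch.
# The event graph also drives WPO cloth material wind and toggles tail hiding for long back bling and skydiving.
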AnimBlueprint /BRCosmetics/Characters/Player/Male/Large/Bodies/M_LRG_BatterBoi/Meshes/M_LRG_BatterBoi_AnimBP
target_skeleton=&/M_LRG_BatterBoi_Skeleton
graphs [
 AnimGraph {nodes [
   n1 = Look At - Bone: dyn_bun!(ComponentPose=Local To Component(LocalPose=Use cached pose 'CR_Start'()), bAlphaBoolEnabled:bool=Get Not_Crouching() -> (Not_Crouching:bool)) -> (Pose) {
    alphaInputType=Bool
    alphaBoolBlend {blendInTime=0.1, blendOption=Linear, blendOutTime=0.1}
    boneToModify {boneName=dyn_bun}
    lookAtLocation=<0 20>
    lookAtTarget {boneReference {boneName=calf_l}}
    lookAt_Axis {axis=<1 0>}
   }
   n2 = "Transform (Modify) Bone - Bone: dyn_bun"!(ComponentPose=@n1, bAlphaBoolEnabled:bool=Get bIsCrouching(), Translation="0,0,0", Rotation="-10,0,0", Scale="1,1,1") -> (Pose) {
    alphaInputType=Bool
    rotationMode=BMM_Additive
    rotationSpace=BCS_BoneSpace
    alphaBoolBlend {blendInTime=0.1, blendOption=Linear, blendOutTime=0.1}
    boneToModify {boneName=dyn_bun}
    rotation {pitch=-10}
   }
   n3 = Get Moving_crouching() -> (Moving_crouching:bool)
   n4 = "Transform (Modify) Bone - Bone: dyn_bun"(ComponentPose=@n2, bAlphaBoolEnabled=@n3, Translation="0,0,0", Rotation="0,-20,0", Scale="1,1,1") {
    alphaInputType=Bool
    rotationMode=BMM_Additive
    rotationSpace=BCS_BoneSpace
    alphaBoolBlend {blendInTime=0.1, blendOption=Linear, blendOutTime=0.1}
    boneToModify {boneName=dyn_bun}
    rotation {yaw=-20}
   }
   n5 = "Transform (Modify) Bone - Bone: dyn_bun_01_l"!(ComponentPose=@n4, bAlphaBoolEnabled:bool=@n3, Translation="0,0,0", Rotation="-15,0,0", Scale="1,1,1") -> (Pose) {
    alphaInputType=Bool
    rotationMode=BMM_Additive
    rotationSpace=BCS_BoneSpace
    alphaBoolBlend {blendInTime=0.1, blendOption=Linear, blendOutTime=0.1}
    boneToModify {boneName=dyn_bun_01_l}
    rotation {pitch=-15}
   }
   n6 = "Transform (Modify) Bone - Bone: dyn_bun_01_r"!(ComponentPose=@n5, bAlphaBoolEnabled:bool=@n3, Translation="0,0,0", Rotation="-15,0,0", Scale="1,1,1") -> (Pose) {
    alphaInputType=Bool
    rotationMode=BMM_Additive
    rotationSpace=BCS_BoneSpace
    alphaBoolBlend {blendInTime=0.1, blendOption=Linear, blendOutTime=0.1}
    boneToModify {boneName=dyn_bun_01_r}
    rotation {pitch=-15}
   }
   n7 = Anim Dynamics - Bone: dyn_bun(ComponentPose=@n6, Alpha=1, ExternalForce="0,0,0") {
    angularDampingOverride=1
    angularSpringConstant=20000
    bAngularSpring=true
    bOverrideAngularDamping=true
    bUseGravityOverride=true
    bUsePlanarLimit=false
    bUseSphericalLimits=true
    gravityScale=0.35
    simulationSpace=RootRelative
    alphaBoolBlend {blendOption=Linear}
    boundBone {boneName=dyn_bun}
    componentLinearAccScale=<1 1 1>
    componentLinearVelScale=<1 1 1>
    gravityOverride {z=-980}
    physicsBodyDefinitions [{
      boundBone {boneName=dyn_bun}
      boxExtents=<40 40 40>
      constraintSetup {angularLimitsMax=<5 10 0>, angularLimitsMin=<-5 -10 -15>, angularTarget=<1 0 0>}
      localJointOffset=<2 0 0>
     }]
    relativeSpaceBone {boneName=hand_l}
    sphericalLimits [{limitRadius=5, drivingBone {boneName=thigh_l}, sphereLocalOffset=<-23.2692 2.0595 0.6415>}]
   }
   n8 = Anim Dynamics - Bone: dyn_top_bun_l(ComponentPose=@n7, Alpha=1, ExternalForce="0,0,0") {
    angularDampingOverride=1
    angularSpringConstant=20000
    bAngularSpring=true
    bOverrideAngularDamping=true
    bUseGravityOverride=true
    bUsePlanarLimit=false
    bUseSphericalLimits=true
    gravityScale=0.35
    simulationSpace=RootRelative
    alphaBoolBlend {blendOption=Linear}
    boundBone {boneName=dyn_top_bun_l}
    componentLinearAccScale=<1 1 1>
    componentLinearVelScale=<1 1 1>
    gravityOverride {z=-980}
    physicsBodyDefinitions [{
      boundBone {boneName=dyn_top_bun_l}
      boxExtents=<40 40 40>
      constraintSetup {angularLimitsMax=<15 15 0>, angularLimitsMin=<-15 -15 -15>, angularTarget=<1 0 0>}
      localJointOffset=<2 0 0>
     }]
    relativeSpaceBone {boneName=hand_l}
    sphericalLimits [{limitRadius=5, drivingBone {boneName=thigh_l}, sphereLocalOffset=<-23.2692 2.0595 0.6415>}]
   }
   n9 = Anim Dynamics - Bone: dyn_top_bun_r(ComponentPose=@n8, Alpha=1, ExternalForce="0,0,0") {
    angularDampingOverride=1
    angularSpringConstant=20000
    bAngularSpring=true
    bOverrideAngularDamping=true
    bUseGravityOverride=true
    bUsePlanarLimit=false
    bUseSphericalLimits=true
    gravityScale=0.35
    simulationSpace=RootRelative
    alphaBoolBlend {blendOption=Linear}
    boundBone {boneName=dyn_top_bun_r}
    componentLinearAccScale=<1 1 1>
    componentLinearVelScale=<1 1 1>
    gravityOverride {z=-980}
    physicsBodyDefinitions [{
      boundBone {boneName=dyn_top_bun_r}
      boxExtents=<40 40 40>
      constraintSetup {angularLimitsMax=<15 15 0>, angularLimitsMin=<-15 -15 -15>, angularTarget=<1 0 0>}
      localJointOffset=<2 0 0>
     }]
    relativeSpaceBone {boneName=hand_l}
    sphericalLimits [{limitRadius=5, drivingBone {boneName=thigh_l}, sphereLocalOffset=<-23.2692 2.0595 0.6415>}]
   }
   n10 = Anim Dynamics - Bone: dyn_bun_01_l(ComponentPose=@n9, Alpha=0.5, ExternalForce="0,0,0") {
    angularDampingOverride=1
    angularSpringConstant=20000
    bOverrideAngularDamping=true
    bUseGravityOverride=true
    bUseSphericalLimits=true
    gravityScale=0.35
    simulationSpace=RootRelative
    alphaBoolBlend {blendOption=Linear}
    boundBone {boneName=dyn_bun_01_l}
    componentAppliedLinearAccClamp=<1000 1000 1000>
    componentLinearAccScale=<1 1 1>
    componentLinearVelScale=<1 1 1>
    gravityOverride {z=-980}
    physicsBodyDefinitions [{
      boundBone {boneName=dyn_bun_01_l}
      boxExtents=<15 15 15>
      constraintSetup {angularLimitsMax=<5 15 0>, angularLimitsMin=<-5 -15 -15>, angularTarget=<1 0 0>}
      localJointOffset=<1 0 0>
     }]
    relativeSpaceBone {boneName=hand_l}
    sphericalLimits [{limitRadius=7, drivingBone {boneName=thigh_l}, sphereLocalOffset=<-21.7731 8.3933 5.2432>}]
   }
   n11 = Anim Dynamics - Bone: dyn_bun_02_l(ComponentPose=@n10, Alpha=0.5, ExternalForce="0,0,0") {
    angularDampingOverride=1
    angularSpringConstant=20000
    bOverrideAngularDamping=true
    bUseGravityOverride=true
    bUseSphericalLimits=true
    gravityScale=0.35
    simulationSpace=RootRelative
    alphaBoolBlend {blendOption=Linear}
    boundBone {boneName=dyn_bun_02_l}
    componentAppliedLinearAccClamp=<1000 1000 1000>
    componentLinearAccScale=<1 1 1>
    componentLinearVelScale=<1 1 1>
    gravityOverride {z=-980}
    physicsBodyDefinitions [{
      boundBone {boneName=dyn_bun_02_l}
      boxExtents=<15 15 15>
      constraintSetup {angularLimitsMax=<5 15 0>, angularLimitsMin=<-5 -15 -15>, angularTarget=<1 0 0>}
      localJointOffset=<1 0 0>
     }]
    relativeSpaceBone {boneName=hand_l}
    sphericalLimits [{limitRadius=7, drivingBone {boneName=thigh_l}, sphereLocalOffset=<-25.8655 4.5001 0.0974>}]
   }
   n12 = Anim Dynamics - Bone: dyn_bun_03_l(ComponentPose=@n11, Alpha=0.5, ExternalForce="0,0,0") {
    angularDampingOverride=1
    angularSpringConstant=20000
    bOverrideAngularDamping=true
    bUseGravityOverride=true
    bUseSphericalLimits=true
    gravityScale=0.35
    simulationSpace=RootRelative
    alphaBoolBlend {blendOption=Linear}
    boundBone {boneName=dyn_bun_03_l}
    componentAppliedLinearAccClamp=<1000 1000 1000>
    componentLinearAccScale=<1 1 1>
    componentLinearVelScale=<1 1 1>
    gravityOverride {z=-980}
    physicsBodyDefinitions [{
      boundBone {boneName=dyn_bun_03_l}
      boxExtents=<15 15 15>
      constraintSetup {angularLimitsMax=<5 15 0>, angularLimitsMin=<-5 -15 -15>, angularTarget=<1 0 0>}
      localJointOffset=<1 0 0>
     }]
    relativeSpaceBone {boneName=hand_l}
    sphericalLimits [{limitRadius=7, drivingBone {boneName=thigh_l}, sphereLocalOffset=<-26.1083 3.4831 -2.926>}]
   }
   n13 = Anim Dynamics - Bone: dyn_bun_01_r(ComponentPose=@n12, Alpha=0.5, ExternalForce="0,0,0") {
    angularDampingOverride=1
    angularSpringConstant=20000
    bOverrideAngularDamping=true
    bUseGravityOverride=true
    bUseSphericalLimits=true
    gravityScale=0.35
    simulationSpace=RootRelative
    alphaBoolBlend {blendOption=Linear}
    boundBone {boneName=dyn_bun_01_r}
    componentAppliedLinearAccClamp=<1000 1000 1000>
    componentLinearAccScale=<1 1 1>
    componentLinearVelScale=<1 1 1>
    gravityOverride {z=-980}
    physicsBodyDefinitions [{
      boundBone {boneName=dyn_bun_01_r}
      boxExtents=<15 15 15>
      constraintSetup {angularLimitsMax=<5 15 0>, angularLimitsMin=<-5 -15 -15>, angularTarget=<1 0 0>}
      localJointOffset=<1 0 0>
     }]
    relativeSpaceBone {boneName=hand_l}
    sphericalLimits [{limitRadius=8, drivingBone {boneName=thigh_l}, sphereLocalOffset=<-23.9127 5.4091 2.0702>}]
   }
   n14 = Anim Dynamics - Bone: dyn_bun_02_r(ComponentPose=@n13, Alpha=0.5, ExternalForce="0,0,0") {
    angularDampingOverride=1
    angularSpringConstant=20000
    bOverrideAngularDamping=true
    bUseGravityOverride=true
    bUseSphericalLimits=true
    gravityScale=0.35
    simulationSpace=RootRelative
    alphaBoolBlend {blendOption=Linear}
    boundBone {boneName=dyn_bun_02_r}
    componentAppliedLinearAccClamp=<1000 1000 1000>
    componentLinearAccScale=<1 1 1>
    componentLinearVelScale=<1 1 1>
    gravityOverride {z=-980}
    physicsBodyDefinitions [{
      boundBone {boneName=dyn_bun_02_r}
      boxExtents=<15 15 15>
      constraintSetup {angularLimitsMax=<5 15 0>, angularLimitsMin=<-5 -15 -15>, angularTarget=<1 0 0>}
      localJointOffset=<1 0 0>
     }]
    relativeSpaceBone {boneName=hand_l}
    sphericalLimits [{limitRadius=8, drivingBone {boneName=thigh_l}, sphereLocalOffset=<-22.2753 3.7002 2.6172>}]
   }
   n15 = Anim Dynamics - Bone: dyn_bun_03_r(ComponentPose=@n14, Alpha=0.5, ExternalForce="0,0,0") {
    angularDampingOverride=1
    angularSpringConstant=20000
    bOverrideAngularDamping=true
    bUseGravityOverride=true
    bUseSphericalLimits=true
    gravityScale=0.35
    simulationSpace=RootRelative
    alphaBoolBlend {blendOption=Linear}
    boundBone {boneName=dyn_bun_03_r}
    componentAppliedLinearAccClamp=<1000 1000 1000>
    componentLinearAccScale=<1 1 1>
    componentLinearVelScale=<1 1 1>
    gravityOverride {z=-980}
    physicsBodyDefinitions [{
      boundBone {boneName=dyn_bun_03_r}
      boxExtents=<15 15 15>
      constraintSetup {angularLimitsMax=<5 15 0>, angularLimitsMin=<-5 -15 -15>, angularTarget=<1 0 0>}
      localJointOffset=<1 0 0>
     }]
    relativeSpaceBone {boneName=hand_l}
    sphericalLimits [{limitRadius=8, drivingBone {boneName=thigh_l}, sphereLocalOffset=<-24.7762 5.4091 -0.4483>}]
   }
   n16 = Component To Local(ComponentPose=@n15)
   n17 = Use cached pose 'CR_RBAN'()
   n18 = Get bRigidBodyEnabled()
   n19 = Anim Dynamics - Bone: dyn_endGloves_l(
    ComponentPose=Local To Component(LocalPose=Blend Poses by bool(BlendPose_0=@n17, BlendPose_1=@n16, bActiveValue=@n18) {blendType=Linear})
    Alpha=1
    ExternalForce="0,0,0"
   ) {
    angularDampingOverride=1
    angularSpringConstant=20000
    bAngularSpring=true
    bOverrideAngularDamping=true
    bUseGravityOverride=true
    gravityScale=0.35
    simulationSpace=RootRelative
    alphaBoolBlend {blendOption=Linear}
    boundBone {boneName=dyn_endGloves_l}
    componentLinearAccScale=<1 1 1>
    componentLinearVelScale=<1 1 1>
    gravityOverride {z=-980}
    physicsBodyDefinitions [{
      boundBone {boneName=dyn_endGloves_l}
      boxExtents=<40 50 50>
      constraintSetup {angularLimitsMax=<15 15 15>, angularLimitsMin=<-15 -15 -15>, angularTarget=<1 0 0>}
      localJointOffset=<4 0 0>
     }]
    relativeSpaceBone {boneName=hand_l}
   }
   n20 = Anim Dynamics - Bone: dyn_endGloves_r(ComponentPose=@n19, Alpha=1, ExternalForce="0,0,0") {
    angularDampingOverride=1
    angularSpringConstant=20000
    bAngularSpring=true
    bOverrideAngularDamping=true
    bUseGravityOverride=true
    gravityScale=0.35
    simulationSpace=RootRelative
    alphaBoolBlend {blendOption=Linear}
    boundBone {boneName=dyn_endGloves_r}
    componentLinearAccScale=<1 1 1>
    componentLinearVelScale=<1 1 1>
    gravityOverride {z=-980}
    physicsBodyDefinitions [{
      boundBone {boneName=dyn_endGloves_r}
      boxExtents=<40 50 50>
      constraintSetup {angularLimitsMax=<15 15 15>, angularLimitsMin=<-15 -15 -15>, angularTarget=<1 0 0>}
      localJointOffset=<4 0 0>
     }]
    relativeSpaceBone {boneName=hand_r}
   }
   n21 = Anim Dynamics - Bone: dyn_tail_01(ComponentPose=@n20, Alpha=1, ExternalForce="0,0,0") {
    angularDampingOverride=1
    angularSpringConstant=20000
    bAngularSpring=true
    bOverrideAngularDamping=true
    bUseGravityOverride=true
    gravityScale=0.35
    simulationSpace=RootRelative
    alphaBoolBlend {blendOption=Linear}
    boundBone {boneName=dyn_tail_01}
    componentAppliedLinearAccClamp=<10000 10000 10000>
    componentLinearAccScale=<1 1 1>
    componentLinearVelScale=<1 1 1>
    gravityOverride {z=-980}
    physicsBodyDefinitions [{
      boundBone {boneName=dyn_tail_01}
      boxExtents=<30 40 40>
      constraintSetup {angularLimitsMax=<7 7 7>, angularLimitsMin=<-7 -7 -7>, angularTarget=<1 0 0>}
      localJointOffset=<4 0 0>
     }]
   }
   Output Pose(Result=Component To Local(ComponentPose=@n21)) {layerGroup=None}
   n22 = Control Rig!(
    Source=Copy Pose From Mesh() {bUseAttachedParent=true}
    Alpha
    isSkydiveFloating:double=Get isSkydiveFloating() -> (isSkydiveFloating:double)
    isLargeGunEquipped:double=Get large_gun_out() -> (large_gun_out:double)
    isParachuteOpen:double=Get isParachuteOpen() -> (isParachuteOpen:double)
   ) -> (Pose) {
    alphaBoolBlend {blendOption=Linear}
    controlRigAssetReference {blueprintRigClass=&/M_LRG_BatterBoi_CtrlRig.M_LRG_BatterBoi_CtrlRig_C}
   }
   Save cached pose 'CR_Start'(Pose=@n22) {cacheName=CR_Start}
   n23 = "Constraint - base_chest (PPPPPPPP)"!(
    ComponentPose=Local To Component(LocalPose=Use cached pose 'CR_Start'())
    Alpha=1
    ConstraintWeights_0=0.6408
    ConstraintWeights_1:float=0.1901
    ConstraintWeights_2:float=0.0533
    ConstraintWeights_3:float=0.0301
    ConstraintWeights_4:float=0.0268
    ConstraintWeights_5:float=0.0267
    ConstraintWeights_6:float=0.0221
    ConstraintWeights_7:float=0.0101
   ) -> (Pose) {
    alphaBoolBlend {blendOption=Linear}
    boneToModify {boneName=base_chest}
    constraintSetup [
     {targetBone {boneName=spine_05}}
     {targetBone {boneName=spine_04}}
     {targetBone {boneName=spine_03}}
     {targetBone {boneName=neck_02}}
     {targetBone {boneName=clavicle_l}}
     {targetBone {boneName=clavicle_r}}
     {targetBone {boneName=head}}
     {targetBone {boneName=spine_02}}
    ]
    constraintWeights [1,1,1,1,1,1,1,1]
   }
   n24 = "Constraint - base_chest_l (PPPPPPPP)"!(
    ComponentPose=@n23
    Alpha=1
    ConstraintWeights_0=0.466
    ConstraintWeights_1:float=0.2583
    ConstraintWeights_2:float=0.0899
    ConstraintWeights_3:float=0.0856
    ConstraintWeights_4:float=0.0596
    ConstraintWeights_5:float=0.0298
    ConstraintWeights_6:float=0.0073
    ConstraintWeights_7:float=0.0035
   ) -> (Pose) {
    alphaBoolBlend {blendOption=Linear}
    boneToModify {boneName=base_chest_l}
    constraintSetup [
     {targetBone {boneName=spine_05}}
     {targetBone {boneName=spine_04}}
     {targetBone {boneName=spine_03}}
     {targetBone {boneName=clavicle_l}}
     {targetBone {boneName=deform_pec_l}}
     {targetBone {boneName=spine_02}}
     {targetBone {boneName=neck_02}}
     {targetBone {boneName=head}}
    ]
    constraintWeights [1,1,1,1,1,1,1,1]
   }
   n25 = "Constraint - base_chest_r (PPPPPPPP)"!(
    ComponentPose=@n24
    Alpha=1
    ConstraintWeights_0=0.466
    ConstraintWeights_1:float=0.2583
    ConstraintWeights_2:float=0.0899
    ConstraintWeights_3:float=0.0856
    ConstraintWeights_4:float=0.0596
    ConstraintWeights_5:float=0.0298
    ConstraintWeights_6:float=0.0073
    ConstraintWeights_7:float=0.0035
   ) -> (Pose) {
    alphaBoolBlend {blendOption=Linear}
    boneToModify {boneName=base_chest_r}
    constraintSetup [
     {targetBone {boneName=spine_05}}
     {targetBone {boneName=spine_04}}
     {targetBone {boneName=spine_03}}
     {targetBone {boneName=clavicle_r}}
     {targetBone {boneName=deform_pec_r}}
     {targetBone {boneName=spine_02}}
     {targetBone {boneName=neck_02}}
     {targetBone {boneName=head}}
    ]
    constraintWeights [1,1,1,1,1,1,1,1]
   }
   n26 = "Constraint - base_pouch (PPPPPP)"!(
    ComponentPose=@n25
    Alpha=1
    ConstraintWeights_0=0.6581
    ConstraintWeights_1:float=0.3158
    ConstraintWeights_2:float=0.024
    ConstraintWeights_3:float=0.0016
    ConstraintWeights_4:float=0.0003
    ConstraintWeights_5:float=0.0002
   ) -> (Pose) {
    alphaBoolBlend {blendOption=Linear}
    boneToModify {boneName=base_pouch}
    constraintSetup [
     {targetBone {boneName=dyn_belt_side_r}}
     {targetBone {boneName=dyn_belt_front_r}}
     {targetBone {boneName=dyn_belt_back_r}}
     {targetBone {boneName=dyn_belt_front_l}}
     {targetBone {boneName=dyn_belt_back_l}}
     {targetBone {boneName=dyn_belt_side_l}}
    ]
    constraintWeights [1,1,1,1,1,1]
   }
   n27 = RigidBody!(ComponentPose=@n26, Alpha, OverrideWorldGravity:Vector="0,0,-980", ExternalForce) -> (Pose) {
    bOverrideWorldGravity=true
    overridePhysicsAsset=&/M_LRG_BatterBoi_Necklace_Physics
    simulationSpace=BaseBoneSpace
    alphaBoolBlend {blendOption=Linear}
    alphaScaleBiasClamp {bInterpResult=true}
    baseBoneRef {boneName=spine_03}
    componentAppliedLinearAccClamp=<3000 3000 3000>
    componentLinearAccScale=<0.3 0.3 0.3>
    componentLinearVelScale=<0.3 0.3 0.3>
    simSpaceSettings {
     maxAngularAcceleration=1000
     maxAngularVelocity=3
     maxLinearAcceleration=700
     maxLinearVelocity=500
     velocityScaleZ=0
     worldAlpha=0.7
     externalLinearDragV=<0.2 0.2>
    }
   }
   n28 = RigidBody(ComponentPose=@n27, OverrideWorldGravity="0,0,-980") {
    bOverrideWorldGravity=true
    overridePhysicsAsset=&/M_LRG_BatterBoi_Physics
    simulationSpace=BaseBoneSpace
    alphaBoolBlend {blendOption=Linear}
    alphaScaleBiasClamp {bInterpResult=true}
    baseBoneRef {boneName=spine_03}
    componentAppliedLinearAccClamp=<3000 3000 3000>
    componentLinearAccScale=<0.3 0.3 0.3>
    componentLinearVelScale=<0.3 0.3 0.3>
    simSpaceSettings {
     maxAngularAcceleration=1000
     maxAngularVelocity=3
     maxLinearAcceleration=700
     maxLinearVelocity=500
     velocityScaleZ=0
     worldAlpha=0.7
     externalLinearDragV=<0.2 0.2>
    }
   }
   Save cached pose 'CR_RBAN'(Pose=Component To Local(ComponentPose=@n28)) {cacheName=CR_RBAN}
  ]}
 EventGraph {nodes [
   Event BlueprintInitializeAnimation()
   > Set SkelMesh(SkelMesh=GetOwnerSkeletalMeshComponent())
   n1 = TryGetPawnOwner()
   > Is Valid(InputObject=@n1)
   > n2 = Cast To FortPlayerPawnAthena(Object=@n1)
   > n3 = Set PlayerPawn(PlayerPawn:Fort Player Pawn Athena Object Reference=@n2) -> (Output_Get:Fort Player Pawn Athena Object Reference)
   > Set Body_Mesh(Body_Mesh=GetSkeletalMeshForPartType(self=@n3))
   n4 = Event OnBeginSkydiving()
   n5 = Get Dynamic Material Instance(execute=@n20.then_1, SkeletalMeshComponent=GetPartSkeletalMeshComponent())
   SetScalarParameterValue(execute=[@n4.then, @n7.else, @n24.then, @n25.else], self=@n5, ParameterName=Hide Element 01, Value=0)
   > PrintText!(execute, InText="INVTEXT(\"Tail hide!\")", bPrintToScreen, bPrintToLog, TextColor, Duration, Key) -> (then)
   Event OnEndSkydiving()
   n6 = CheckForBackbling()
   > n7 = Branch(Condition=@n6.bIsBackblingLong)
   SetScalarParameterValue(execute=[@n7.then, @n24.else], self=@n5, ParameterName=Hide Element 01, Value=1)
   n8 = Event BlueprintUpdateAnimation()
   > n9 = Sequence!(execute) -> (then_0, then_1, then_2:Exec, then_3:Exec, then_4:Exec, then_5:Exec)
   > Is Valid(InputObject=Get PlayerPawn() -> (PlayerPawn:Fort Player Pawn Athena Object Reference))
   > n10 = Branch(Condition=Get bIsInFrontEnd())
   Set WPO Globals Function(
    execute=@n10.else
    Material=Get Mat Instance Dynamic()
    Overlay Material=Get Overlay Mat Instance Dynamic()
    Skeletal Mesh=Get SkelMesh()
    "Local Noise Transform  Location"=Get Local Noise Transform Location() -> (Local Noise Transform Location:Vector)
    Local Noise Offset=Get Local Noise Offset()
   )
   > Set WPO Regional Function!(
    execute
    Material:Material Instance Dynamic Object Reference=Get Mat Instance Dynamic()
    Overlay Material:Material Instance Dynamic Object Reference=Get Overlay Mat Instance Dynamic()
    Skeletal Mesh:Skeletal Mesh Component Object Reference=Get SkelMesh()
    Velocity Transform:Name=foot_r
    Velocity Interp Speed:double=Get Velocity Interp Speed()
    Collision Transform:Name=calf_r
    Collision Transform Axis:EAxis Enum
    Invert Collision Transform Axis:bool=false
    Delta Time X:double=@n8.DeltaTimeX
    Previous Velocity Transform Location:"Vector (by ref)"=Get R1 Velocity Transform Location()
    Previous Velocity Transform Velocity:"Vector (by ref)"=Get R1 Velocity Transform Velocity()
    Region:int=1
   ) -> (then)
   > Set WPO Regional Function(
    Material=Get Mat Instance Dynamic()
    Overlay Material=Get Overlay Mat Instance Dynamic()
    Skeletal Mesh=Get SkelMesh()
    Velocity Transform=foot_l
    Velocity Interp Speed=Get Velocity Interp Speed()
    Collision Transform=calf_l
    Invert Collision Transform Axis=true
    Delta Time X=@n8.DeltaTimeX
    Previous Velocity Transform Location=Get R2 Velocity Transform Location()
    Previous Velocity Transform Velocity=Get R2 Velocity Transform Velocity()
    Region=2
   )
   > n11 = Branch(Condition=OR Boolean(A=Get bIsSkydiveFloating(), B=Get bIsParachuteOpened()))
   > Set WPO Pho Wind Intensity(WPO Pho Wind Intensity=0.8)
   > n12 = Set WPO Pho Turbulence(WPO Pho Turbulence:double=1.2) -> (Output_Get:double)
   SetScalarParameterValue(
    execute=[@n14.then, @n12.then]
    self=[Get Mat Instance Dynamic(), Get Overlay Mat Instance Dynamic()]
    ParameterName=ClothSim_WindStrength
    Value=Get WPO Pho Wind Intensity()
   )
   n13 = Get WPO Pho Turbulence() -> (WPO Pho Turbulence:double)
   > SetScalarParameterValue(self=[Get Mat Instance Dynamic(), Get Overlay Mat Instance Dynamic()], ParameterName=ClothSim_R1_TurbulenceStrength, Value=@n13)
   > SetScalarParameterValue(self=[Get Mat Instance Dynamic(), Get Overlay Mat Instance Dynamic()], ParameterName=ClothSim_R2_TurbulenceStrength, Value=@n13)
   Set WPO Pho Wind Intensity(execute=@n11.else, WPO Pho Wind Intensity=0)
   > n14 = Set WPO Pho Turbulence(WPO Pho Turbulence=0.1)
   n15 = Gravity Override State Function DEPRECATED(
    execute=@n9.then_1
    CharacterPartAnimInstance=Self-Reference()
    GameTimeInSeconds=GetGameTimeInSeconds()
    GravityOverrideParams=Get Gravity Override Params()
   )
   > Set GravityOverride(GravityOverride=@n15)
   > Set Not_Crouching(Not_Crouching:bool=NOT Boolean(A=Get bIsCrouching())) -> (Output_Get:bool)
   > Set Moving_crouching(Moving_crouching:bool=AND Boolean(A=Get bIsCrouching(), B=Get bIsMoving2D())) -> (Output_Get:bool)
   n16 = Branch(execute=@n9.then_2, Condition=OR Boolean(A=Get bIsSkydiveDiveMode(), B=Get bIsSkydiveFloating()))
   > Set isSkydiveFloating(isSkydiveFloating:double=1) -> (Output_Get:double)
   Set isSkydiveFloating(execute=@n16.else, isSkydiveFloating=0)
   n17 = Cast To FortPlayerAnimInstance(execute=@n9.then_3, Object=GetAnimInstance(self=GetOwnerSkeletalMeshComponent()))
   > Set rifle_out(rifle_out:bool=false) -> (Output_Get:bool)
   > n18 = Switch on EFortWeaponCoreAnimation!(execute, Selection:EFortWeaponCoreAnimation Enum=Get WeaponCoreAnim(self=@n17)) -> (
    Melee:Exec
    Pistol:Exec
    Shotgun:Exec
    PaperBlueprint:Exec
    Rifle:Exec
    MeleeOneHand:Exec
    MachinePistol:Exec
    RocketLauncher:Exec
    GrenadeLauncher:Exec
    GoingCommando:Exec
    AssaultRifle:Exec
    TacticalShotgun:Exec
    SniperRifle:Exec
    TrapPlacement:Exec
    ShoulderLauncher:Exec
    AbilityDecoTool:Exec
    Crossbow:Exec
    C4:Exec
    RemoteControl:Exec
    DualWield:Exec
    AR_BullPup:Exec
    AR_ForwardGrip:Exec
    MedPackPaddles:Exec
    SMG_P90:Exec
    AR_DrumGun:Exec
    Consumable_Small:Exec
    Consumable_Large:Exec
    Balloon:Exec
    MountedTurret:Exec
    CreativeTool:Exec
    ExplosiveBow:Exec
    AshtonIndigo:Exec
    AshtonChicago:Exec
    MeleeDualWield:Exec
    Unarmed:Exec
   )
   Branch(execute=@n18.AssaultRifle)
   > Set rifle_out(rifle_out=true)
   Set large_gun_out(execute=@n9.then_4, large_gun_out:double="To Float (Boolean)"(InBool=Get rifle_out() -> (rifle_out:bool))) -> (Output_Get:double)
   n19 = Branch(execute=@n9.then_5, Condition=Get bIsParachuteOpened())
   > Set isParachuteOpen(isParachuteOpen:double=1) -> (Output_Get:double)
   Set isParachuteOpen(execute=@n19.else, isParachuteOpen=0)
   Event OnCharacterCustomizationLoaded()
   > n20 = Sequence!(execute) -> (then_0, then_1)
   > n21 = Set SkelMesh(SkelMesh=GetPartSkeletalMeshComponent())
   > n22 = Get Dynamic Material Instance(SkeletalMeshComponent=@n21)
   > Set Mat Instance Dynamic(Mat Instance Dynamic=@n22)
   > n23 = GetOrCreateDynamicOverlayMaterial(MeshComponent=@n21, MaterialName=None)
   > Set Overlay Mat Instance Dynamic(Overlay Mat Instance Dynamic=@n23)
   PickaxeItemDefinitionHasAnyMetaTag(Tags=Make Array("[0]"=(TagName="Cosmetics.MetaTags.BackSpin")), bRequiresEquipped=true)
   Event OnWeaponChanged()
   K2Node_Event_8 = CustomEvent!() -> (
    OutputDelegate
    then
    CustomBaseMesh:Skeletal Mesh Component Object Reference
    CustomParts:Array of Custom Character Part Object References
    CustomMeshParts:Array of Skeletal Mesh Component Budgeted Object References
   ) {is=title K2Node_CustomEvent}
   Reroute Node(InputPin:Wildcard) -> (OutputPin:Wildcard)
   n24 = Branch(execute=@n25.then, Condition=Get bIsSkyDiving())
   n25 = Branch(execute=@n5.then, Condition=@n6.bIsBackblingLong)
  ]}
]
variables [
 Body_Mesh {type=SkeletalMeshComponent}
 Gravity Override Params {type=GravityOverrideParamsStruct}
 GravityOverride {type=Vector}
 isParachuteOpen {type=number}
 isSkydiveFloating {type=number}
 large_gun_out {type=number}
 Local Noise Offset {type=Vector}
 Local Noise Transform Location {type=Vector}
 Mat Instance Dynamic {type=MaterialInstanceDynamic}
 Moving_crouching {type=boolean}
 Not_Crouching {type=boolean}
 Overlay Mat Instance Dynamic {type=MaterialInstanceDynamic}
 PlayerPawn {type=FortPlayerPawnAthena}
 R1 Velocity Transform Location {type=Vector}
 R1 Velocity Transform Velocity {type=Vector}
 R2 Velocity Transform Location {type=Vector}
 R2 Velocity Transform Velocity {type=Vector}
 rifle_out {type=boolean}
 SkelMesh {type=SkeletalMeshComponent}
 Velocity Interp Speed {type=number}
 WPO Pho Turbulence {type=number}
 WPO Pho Wind Intensity {type=number}
]
