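# Physics asset of cape colliders for the large male BatterBoi player body, apparently for the cape to collide against.
# Capsules cover the pelvis, spine, clavicles, upper neck, rear belt and pouch, with a sphere on the head and several capsules layered on spine_02 and spine_05.
# Arm, leg and foot bodies are present but carry no shapes.
# It defines no constraints, so it provides collision volumes without a jointed ragdoll setup.
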PhysicsAsset /BRCosmetics/Characters/Player/Male/Large/Bodies/M_LRG_BatterBoi/Meshes/M_LRG_BatterBoi_CapeColliders
bodies {
 base_belt_back_l [{shapeType=Capsule, length=20.1541, radius=5.1592, shapeName=base_belt_back_l_capsule, rotation=<rot -5.4568 -89.9999 0>}]
 base_belt_back_r [{shapeType=Capsule, length=20.1541, radius=5.1592, shapeName=base_belt_back_r_capsule, rotation=<rot -5.4568 -89.9999 0>}]
 base_pouch [{
   shapeType=Capsule
   length=3.047
   radius=7.4741
   shapeName=dyn_pouch_capsule
   center=<4.5613 -3.7688 0.0453>
   rotation=<rot -1.0374 -99.6242 -96.8564>
  }]
 calf_l []
 calf_r []
 clavicle_l [{shapeType=Capsule, length=9.0192, radius=9.8871, center=<9.4266 -1.1806 -2.4977>, rotation=<rot 69.3297 1.4476 176.9462>}]
 clavicle_r [{shapeType=Capsule, length=8.8031, radius=11.892, center=<-9.4266 1.1806 2.4977>, rotation=<rot 69.3297 1.4476 -3.0538>}]
 foot_l []
 foot_r []
 head [{shapeType=Sphere, radius=17.3125, center=<11.028 -0.174 0>}]
 lowerarm_l []
 lowerarm_r []
 neck_02 [{shapeType=Capsule, length=11.7376, radius=9.0174, center=<0.9184 -5.525 0>, rotation=<rot 90 0 -14.6828>}]
 pelvis [{shapeType=Capsule, length=7.5437, radius=15.6406, center=<-2.4887 0.5656 0>}]
 spine_01 [{shapeType=Capsule, length=18.7672, radius=15.2262, center=<-3.0295 0.9366 0>}]
 spine_02 [
  {shapeType=Capsule, length=11.1121, radius=12.4637, center=<-1.711 -2.9947 0.0001>}
  {shapeType=Capsule, length=29.3956, radius=11.5807, center=<2.5024 0.4992 3.9825>, rotation=<rot 90 0.0049 0.0049>}
  {shapeType=Capsule, length=29.875, radius=11.1013, center=<2.4704 0.4991 -3.9825>, rotation=<rot 90 -0.0396 -0.0396>}
 ]
 spine_03 [{shapeType=Capsule, length=12.0269, radius=13.6457, center=<0.1445 -1.2631 0.0001>}]
 spine_04 [{shapeType=Capsule, length=12.939, radius=17.8046, center=<6.3789 -1.8452 0.0002>}]
 spine_05 [
  {shapeType=Capsule, length=17.4141, radius=15.9189, center=<0.797 -2.81 0>}
  {shapeType=Capsule, length=13.9518, radius=13.9425, center=<-0.2916 0.4452 0>}
 ]
 thigh_l []
 thigh_r []
 upperarm_l []
 upperarm_r []
}
constraints []
